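# Editor utility widget that swaps materials on selected level actors from old season content paths to matching assets under new paths.
# It substitutes editable old and new path prefixes in each static mesh material's path and applies the replacement only if that asset exists.
# Changes run in a transaction, with optional checkout and saving of modified actors.
# Other buttons filter the selection to affected actors or add unmatched material folders as old paths.
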EditorUtilityWidgetBlueprint /FigmentTools/SeasonMatReplace/EUW_SeasonMatReplace
graphs [
 AddPath {nodes [
   n1 = AddPath!() -> (then, NewItem:"String (by ref)", bIsNewPath:bool=~) {is=title K2Node_FunctionEntry}
   > n2 = Branch(Condition=@n1.bIsNewPath)
   > n3 = Add Unique(TargetArray=Get NewPaths() -> (NewPaths:Array of Strings), NewItem=@n1)
   UpdatePathText(execute=[@n3.then, @n4.then])
   n4 = Add Unique(execute=@n2.else, TargetArray=Get OldPaths() -> (OldPaths:Array of Strings), NewItem=@n1)
  ]}
 ClearPaths {nodes [ClearPaths() {is=title K2Node_FunctionEntry}]}
 EventGraph {nodes [
   "On Clicked (Button_ReplaceMatsOnMesh)"() {is=title K2Node_ComponentBoundEvent}
   > Clear(TargetArray:Array of Actor Object References=Get ModifiedActors() -> (ModifiedActors:Array of Actor Object References))
   > SetDebugText!(execute, self, InText:Text="LOCTEXT(\"Trying to Replace Materials\")") -> (then)
   > n1 = Create Transaction!(
    Begin:Exec
    Finalize:Exec=@n4.then
    Cancel:Exec=@n4.else
    Context:String=Repalce Materials
    Description:Text="LOCTEXT(\"Repalce Materials\")"
   ) -> (Pending:Exec, Complete:Exec) {is=title K2Node_MacroInstance}
   SetDebugText(execute=@n1.Complete, InText="LOCTEXT(\"Saving Modified Assets... Please Wait\")")
   > Delay!(execute, WorldContextObject:Object Reference, Duration=2) -> (then)
   > SaveAndCheckoutModifiedActors()
   n2 = GetSelectedLevelActors(execute=@n1.Pending, self=Get EditorActorSubsystem())
   > n3 = ReplaceDataOnSelected!(execute, self, Array:Array of Actor Object References=@n2) -> (then, NumActorsModified:int=0)
   > n4 = Branch(Condition="integer > integer"(A=@n3))
   "On Clicked (Button_FilterSelection)"() {is=title K2Node_ComponentBoundEvent}
   > n5 = FilterSelectedActorsToReplace!(execute, self) -> (then, FilteredActors:Array of Actor Object References)
   > SetSelectedLevelActors(self=Get EditorActorSubsystem(), ActorsToSelect=@n5)
   "On Clicked (Button_AddPath)"() {is=title K2Node_ComponentBoundEvent}
   n6 = Get OldSeasonTextBox() -> (OldSeasonTextBox:Editor Utility Editable Text Object Reference)
   > AddPath!(execute, self, NewItem:"String (by ref)"="To String (Text)"(InText=Get Text(self:Editable Text Object Reference=@n6)), bIsNewPath:bool=false) -> (then) {is=title K2Node_CallFunction}
   "On Clicked (Button_DeleatePath)"() {is=title K2Node_ComponentBoundEvent}
   n7 = "To String (Text)"(InText=Get Text(self=Get OldSeasonTextBox()))
   n8 = Get OldPaths()
   > Branch(Condition=Contains Item(TargetArray:Array of Strings=@n8, ItemToFind:"String (by ref)"=@n7))
   > Remove Item!(execute, TargetArray=@n8, Item:"String (by ref)"=@n7) -> (then, ReturnValue:bool)
   > UpdatePathText()
   "On Clicked (Button_AddPath_New)"() {is=title K2Node_ComponentBoundEvent}
   n9 = Get NewSeasonTextBox() -> (NewSeasonTextBox:Editor Utility Editable Text Object Reference)
   > AddPath(NewItem="To String (Text)"(InText=Get Text(self=@n9)), bIsNewPath=true)
   "On Clicked (Button_DeleatePath_New)"() {is=title K2Node_ComponentBoundEvent}
   n10 = "To String (Text)"(InText=Get Text(self=Get NewSeasonTextBox()))
   n11 = Get NewPaths()
   > Branch(Condition=Contains Item(TargetArray=@n11, ItemToFind=@n10))
   > Remove Item(TargetArray=@n11, Item=@n10)
   > UpdatePathText()
   "On Clicked (Button_AddOldPathFromMesh)"() {is=title K2Node_ComponentBoundEvent}
   > n12 = GetSelectedLevelActors(self=Get EditorActorSubsystem())
   > n13 = For Each Loop(Array:Array of Actor Object References=@n12) -> (Array Element:Actor Object Reference)
   > n14 = For Each Loop(Array:Array of Static Mesh Component Object References=Get Components By Class!(self:Actor Object Reference=@n13, ComponentClass:Actor Component Class Reference=&StaticMeshComponent) -> (ReturnValue:Array of Static Mesh Component Object References)) -> (Array Element:Static Mesh Component Object Reference)
   > n15 = For Each Loop(Array:Array of Material Interface Object References=GetMaterials!(self:Mesh Component Object Reference=@n14) -> (ReturnValue:Array of Material Interface Object References)) -> (Array Element:Material Interface Object Reference)
   n16 = Get Object Path String(Object=@n15)
   > n17 = FindValidateAssetReplacement!(execute, self, SourceObjectPath:String=@n16) -> (then, AssetData:Asset Data Structure, bValid:bool)
   > n18 = Branch(Condition=@n17.bValid)
   n19 = GetPath!(InPath:String=@n16) -> (ReturnValue:String)
   PrintString(execute=@n18.else, InString=@n19)
   > AddPath(NewItem=Append(A=@n19, B=/), bIsNewPath=false)
   Event Construct()
   > Append Array!(execute, TargetArray=Get OldPaths(), SourceArray:Array of Strings=Make Array("[0]":String=/Game/) -> (Array:Array of Strings)) -> (then)
   > UpdatePathText()
   Event PreConstruct()
  ]}
 FilterSelectedActorsToReplace {nodes [
   FilterSelectedActorsToReplace() {is=title K2Node_FunctionEntry}
   > n1 = GetSelectedLevelActors(self=Get EditorActorSubsystem())
   > n2 = For Each Loop(Array:Array of Actor Object References=@n1) -> (Array Element:Actor Object Reference)
   Return Node!(execute=@n2.Completed, FilteredActors:Array of Actor Object References=Get ActorList() -> (ActorList:Array of Actor Object References))
   > n3 = Set CurrentActor(CurrentActor:Actor Object Reference=@n2) -> (Output_Get:Actor Object Reference)
   > n4 = Cast To StaticMeshActor!(execute, Object=@n3) -> (then, CastFailed, AsStatic Mesh Actor:Static Mesh Actor Object Reference) {is=title K2Node_DynamicCast}
   > n5 = Cast To StaticMeshComponent(Object=Get RootComponent!(self:Actor Object Reference=@n4) -> (RootComponent:Scene Component Object Reference))
   > n6 = For Each Loop with Break!(Exec:Exec, Array:Array of Material Interface Object References=GetMaterials(self=@n5), Break:Exec=@n9.then) -> (LoopBody:Exec, Array Element:Material Interface Object Reference, Array Index:int, Completed:Exec)
   > n7 = ValidateStringAgainstOldPaths!(execute, self, SourceObjectPath:String=Get Object Path String(Object=@n6), TargetAssetPath:String) -> (then, bShouldBeReplaced:bool, NewAssetPath:String)
   > Branch(Condition=@n7)
   > Add(
    TargetArray:Array of Actor Object References=Get ActorList()
    NewItem:"Actor Object Reference (by ref)"=Get CurrentActor() -> (CurrentActor:Actor Object Reference)
   ) {is=title K2Node_CallArrayFunction}
   n8 = Get CurrentActor()
   > n9 = PrintString(InString=Append(A="Actor Needs Material replaced: ", B=GetDisplayName(Object=@n8)))
   PrintString(execute=@n5.CastFailed, InString=Not a SMC)
  ]}
 FindValidateAssetReplacement {nodes [
   n1 = FindValidateAssetReplacement!() -> (then, SourceObjectPath:String) {is=title K2Node_FunctionEntry}
   > n2 = For Each Loop(Array:Array of Strings=Get NewPaths()) -> (Array Element:String)
   Return Node!(execute=@n2.Completed, AssetData:Asset Data Structure, bValid:bool=false)
   > n3 = ValidateStringAgainstOldPaths(SourceObjectPath=@n1, TargetAssetPath=@n2)
   > Branch(Condition=@n3)
   > n4 = FindAssetData!(execute, AssetPath:String=@n3.NewAssetPath) -> (then, ReturnValue:Asset Data Structure)
   > Branch(Condition=Is Valid Asset Data!(InAssetData:"Asset Data Structure (by ref)"=@n4) -> (ReturnValue:bool))
   > Return Node(AssetData=@n4, bValid=true)
  ]}
 ReplaceDataOnSelected {nodes [
   n1 = ReplaceDataOnSelected!() -> (then, Array:Array of Actor Object References) {is=title K2Node_FunctionEntry}
   > n2 = For Each Loop(Array:Array of Actor Object References=@n1) -> (Array Element:Actor Object Reference)
   Return Node!(execute=@n2.Completed, NumActorsModified:int=Get ChangesMade() -> (ChangesMade:int=0))
   > n3 = Set Actor(Actor:Actor Object Reference=@n2) -> (Output_Get:Actor Object Reference)
   > n4 = For Each Loop(Array:Array of Static Mesh Component Object References=Get Components By Class(self=@n3, ComponentClass=&StaticMeshComponent)) -> (Array Element:Static Mesh Component Object Reference)
   > n5 = Set TargetComponent(TargetComponent:Static Mesh Component Object Reference=@n4) -> (Output_Get:Static Mesh Component Object Reference)
   > ReplaceMaterialsOnComponent!(
    execute
    self
    ActorTarget:Actor Object Reference=Get Actor() -> (Actor:Actor Object Reference)
    StaticMeshComponent:Mesh Component Object Reference=@n5
   ) -> (then)
  ]}
 ReplaceMaterialsOnComponent {nodes [
   n1 = ReplaceMaterialsOnComponent!() -> (then, ActorTarget:Actor Object Reference, StaticMeshComponent:Mesh Component Object Reference) {is=title K2Node_FunctionEntry}
   > Set MadeChangesToActor(MadeChangesToActor:bool=false) -> (Output_Get:bool)
   > n2 = For Each Loop(Array:Array of Material Interface Object References=GetMaterials(self=@n1.StaticMeshComponent)) -> (Array Element:Material Interface Object Reference)
   Branch(execute=@n2.Completed, Condition=Get MadeChangesToActor() -> (MadeChangesToActor:bool))
   > Add Unique(TargetArray:Array of Actor Object References=Get ModifiedActors(), NewItem:"Actor Object Reference (by ref)"=@n1)
   > Increment Int(Value=Get ChangesMade() -> (ChangesMade=0))
   > SaveAndCheckoutActor!(execute, self, Actor:Actor Object Reference=@n1) -> (then)
   > Set MadeChangesToActor(MadeChangesToActor=false)
   Set Index(execute=@n2.LoopBody, Index:int=@n2.Array Index) -> (Output_Get:int=0)
   > n3 = FindValidateAssetReplacement(SourceObjectPath=Get Object Path String(Object=@n2)) {is=title K2Node_CallFunction}
   > Branch(Condition=@n3.bValid)
   n4 = GetAsset(InAssetData=@n3)
   > PrintString(InString=GetDisplayName(Object=@n4))
   > TransactObject(Object=Get Actor())
   > n5 = Cast To MaterialInterface!(execute, Object=@n4) -> (then, CastFailed, AsMaterial Interface:Material Interface Object Reference) {is=title K2Node_DynamicCast}
   > SetMaterial(
    self=Get StaticMeshComponent!() -> (StaticMeshComponent:Mesh Component Object Reference)
    ElementIndex=Get Index() -> (Index:int=0)
    Material=@n5
   )
   > PrintString(InString=Updated Material on Mesh)
   > Set MadeChangesToActor(MadeChangesToActor=true)
   PrintString(InString=No valid Asset Found at Path)
  ]}
 SaveAndCheckoutActor {nodes [
   n1 = SaveAndCheckoutActor!() -> (then, Actor:Actor Object Reference) {is=title K2Node_FunctionEntry}
   > n2 = Branch(Condition=IsChecked(self=Get EditorUtilityCheckBox_Checkout() -> (EditorUtilityCheckBox_Checkout:Editor Utility Check Box Object Reference)))
   > n3 = CheckoutLoadedAsset!(execute, AssetToCheckout:Object Reference=@n1) -> (then, ReturnValue:bool)
   Branch(
    execute=[@n3.then, @n2.else]
    Condition=IsChecked(self=Get EditorUtilityCheckBox_Save() -> (EditorUtilityCheckBox_Save:Editor Utility Check Box Object Reference))
   )
   SaveLoadedAsset(AssetToSave=@n1, bOnlyIfIsDirty=false)
   Branch(Condition=IsChecked(self=Get EditorUtilityCheckBox_Save()))
  ]}
 SaveAndCheckoutModifiedActors {nodes [
   SaveAndCheckoutModifiedActors() {is=title K2Node_FunctionEntry}
   > Branch(Condition=IsChecked(self=Get EditorUtilityCheckBox_Save()))
   > SaveLoadedAssets(AssetsToSave=Get ModifiedActors(), bOnlyIfIsDirty=false)
   > SetDebugText(InText="LOCTEXT(\"Saved!\")")
   Branch(Condition=IsChecked(self=Get EditorUtilityCheckBox_Save()))
  ]}
 SetDebugText {nodes [
   n1 = SetDebugText!() -> (then, InText:Text) {is=title K2Node_FunctionEntry}
   > "SetText (Text)"(self=Get TextBlock_Debug() -> (TextBlock_Debug:Text Object Reference), InText=@n1)
   > PrintString(InString="To String (Text)"(InText=@n1), bPrintToScreen=false, Key="Rep[laceBug")
  ]}
 UpdatePathText {nodes [
   UpdatePathText() {is=title K2Node_FunctionEntry}
   > n1 = For Each Loop(Array:Array of Strings=Get OldPaths()) -> (Array Element:String)
   Set Text(execute=@n1.Completed, Text=Get NewText() -> (NewText:Text), self=Get TextBlock_Paths() -> (TextBlock_Paths:Text Object Reference))
   > Set NewText(NewText:Text) -> (Output_Get:Text)
   > n2 = For Each Loop(Array=Get NewPaths())
   Set Text(execute=@n2.Completed, Text=Get NewText(), self=Get TextBlock_PathsNew() -> (TextBlock_PathsNew:Text Object Reference))
   Set NewText(execute=@n2.LoopBody, NewText=Format Text!(Format="LOCTEXT(\"{Old}\\r\\n{x}\")", Old:Text=Get NewText(), x:String=@n2) -> (Result))
   Set NewText(execute=@n1.LoopBody, NewText=Format Text(Format="LOCTEXT(\"{Old}\\r\\n{x}\")", Old=Get NewText(), x=@n1))
  ]}
 ValidateStringAgainstOldPaths {nodes [
   ValidateStringAgainstOldPaths!() -> (then, SourceObjectPath:String, TargetAssetPath:String) {is=title K2Node_FunctionEntry}
   > n1 = For Each Loop(Array=Get OldPaths())
   PrintString(execute=@n1.Completed, InString=Invalid Path)
   > Return Node!(execute, bShouldBeReplaced:bool=false, NewAssetPath:String)
   n2 = Get TargetAssetPath() -> (TargetAssetPath:String)
   n3 = Branch(execute=@n1.LoopBody, Condition=Contains(SearchIn=Get SourceObjectPath() -> (SourceObjectPath:String), Substring=@n2))
   > Return Node(bShouldBeReplaced=false)
   Branch(execute=@n3.else, Condition=Contains(SearchIn=Get SourceObjectPath(), Substring=@n1))
   > PrintString(InString=Get SourceObjectPath())
   > Return Node(bShouldBeReplaced=true, NewAssetPath=Replace(SourceString=Get SourceObjectPath(), From=@n1, To=@n2))
  ]}
]
variables [
 ModifiedActors {type="array<Actor>"}
 NewPaths {type="array<string>", default [/FigmentEnvironment/LM_8_1_1/, /FigmentEnvironment/]}
 OldPaths {type="array<string>", default [/Figment_S07/LM_7_00/, /Figment_S07/, /Figment_S08/, /Figment_S09/]}
]
widgets {
 info {rootWidgetClass=&Overlay, widgetCount=38}
 widgets [
  Overlay_0 {widgetClassPath=&Overlay, widget=Overlay_0}
  VerticalBox_89 {
   widgetClassPath=&VerticalBox
   parent=Overlay_0
   widget=VerticalBox_89
   slot {class=OverlaySlot, properties {horizontalAlignment=HAlign_Fill, verticalAlignment=VAlign_Fill}}
  }
  HorizontalBox_61 {
   widgetClassPath=&HorizontalBox
   parent=VerticalBox_89
   widget=HorizontalBox_61
   slot {class=VerticalBoxSlot, properties {padding=<ltrb 4 4 4 4>}}
  }
  Border_0 {
   widgetClassPath=&Border
   parent=HorizontalBox_61
   widget=Border_0
   properties {clipping=ClipToBounds, brushColor=<rgba 0 0 0 0.5>, padding=<ltrb 2 2 2 2>}
   slot {class=HorizontalBoxSlot, properties {size {sizeRule=Fill, value=1}}}
  }
  NewSeasonTextBox {
   widgetClassPath=/Script/Blutility.EditorUtilityEditableText
   bIsVariable=true
   parent=Border_0
   widget=NewSeasonTextBox
   properties {hintText=New Season Path, text=&LM_8_1_1/}
   slot {class=BorderSlot, properties {padding=<ltrb 2 2 2 2>}}
  }
  Button_AddPath_New {
   widgetClassPath=&EditorUtilityButton
   parent=HorizontalBox_61
   widget=Button_AddPath_New
   slot {class=HorizontalBoxSlot, properties {horizontalAlignment=HAlign_Left, verticalAlignment=VAlign_Top}}
  }
  TextBlock_5 {
   widgetClassPath=&TextBlock
   parent=Button_AddPath_New
   widget=TextBlock_5
   properties {
    text=Add New Path Reference
    font {
     bForceMonospaced=false
     bMaterialIsStencil=false
     fontMaterial=null
     fontObject=&Roboto
     letterSpacing=0
     monospacedWidth=1
     size=9
     skewAmount=0
     typefaceFontName=Regular
     outlineSettings {
      bApplyOutlineToDropShadows=false
      bMiteredCorners=false
      bSeparateFillAlpha=false
      outlineMaterial=null
      outlineSize=0
      outlineColor=<rgba 0 0 0 1>
     }
    }
   }
   slot {class=ButtonSlot, properties {padding=<ltrb 2 2 2 2>}}
  }
  Button_DeleatePath_New {
   widgetClassPath=&EditorUtilityButton
   parent=HorizontalBox_61
   widget=Button_DeleatePath_New
   slot {class=HorizontalBoxSlot, properties {horizontalAlignment=HAlign_Left, verticalAlignment=VAlign_Top}}
  }
  TextBlock_6 {
   widgetClassPath=&TextBlock
   parent=Button_DeleatePath_New
   widget=TextBlock_6
   properties {
    text=Delete New Path Reference
    font {
     bForceMonospaced=false
     bMaterialIsStencil=false
     fontMaterial=null
     fontObject=&Roboto
     letterSpacing=0
     monospacedWidth=1
     size=9
     skewAmount=0
     typefaceFontName=Regular
     outlineSettings {
      bApplyOutlineToDropShadows=false
      bMiteredCorners=false
      bSeparateFillAlpha=false
      outlineMaterial=null
      outlineSize=0
      outlineColor=<rgba 0 0 0 1>
     }
    }
   }
   slot {class=ButtonSlot, properties {padding=<ltrb 2 2 2 2>}}
  }
  HorizontalBox {
   widgetClassPath=&HorizontalBox
   parent=VerticalBox_89
   widget=HorizontalBox
   slot {class=VerticalBoxSlot, properties {padding=<ltrb 4 4 4 4>}}
  }
  Border {
   widgetClassPath=&Border
   parent=HorizontalBox
   widget=Border
   properties {clipping=ClipToBounds, brushColor=<rgba 0 0 0 0.5>, padding=<ltrb 2 2 2 2>}
   slot {class=HorizontalBoxSlot, properties {size {sizeRule=Fill, value=1}}}
  }
  OldSeasonTextBox {
   widgetClassPath=&EditorUtilityEditableText
   bIsVariable=true
   parent=Border
   widget=OldSeasonTextBox
   properties {hintText=Old Season Path, text=&LM_7_00/}
   slot {class=BorderSlot, properties {padding=<ltrb 2 2 2 2>}}
  }
  Button_AddPath {
   widgetClassPath=&EditorUtilityButton
   parent=HorizontalBox
   widget=Button_AddPath
   slot {class=HorizontalBoxSlot, properties {horizontalAlignment=HAlign_Left, verticalAlignment=VAlign_Top}}
  }
  TextBlock_1 {
   widgetClassPath=&TextBlock
   parent=Button_AddPath
   widget=TextBlock_1
   properties {
    text=Add Old Path Reference
    font {
     bForceMonospaced=false
     bMaterialIsStencil=false
     fontMaterial=null
     fontObject=&Roboto
     letterSpacing=0
     monospacedWidth=1
     size=9
     skewAmount=0
     typefaceFontName=Regular
     outlineSettings {
      bApplyOutlineToDropShadows=false
      bMiteredCorners=false
      bSeparateFillAlpha=false
      outlineMaterial=null
      outlineSize=0
      outlineColor=<rgba 0 0 0 1>
     }
    }
   }
   slot {class=ButtonSlot, properties {padding=<ltrb 2 2 2 2>}}
  }
  Button_DeleatePath {
   widgetClassPath=&EditorUtilityButton
   parent=HorizontalBox
   widget=Button_DeleatePath
   slot {class=HorizontalBoxSlot, properties {horizontalAlignment=HAlign_Left, verticalAlignment=VAlign_Top}}
  }
  TextBlock_3 {
   widgetClassPath=&TextBlock
   parent=Button_DeleatePath
   widget=TextBlock_3
   properties {
    text=Delete Old Path Reference
    font {
     bForceMonospaced=false
     bMaterialIsStencil=false
     fontMaterial=null
     fontObject=&Roboto
     letterSpacing=0
     monospacedWidth=1
     size=9
     skewAmount=0
     typefaceFontName=Regular
     outlineSettings {
      bApplyOutlineToDropShadows=false
      bMiteredCorners=false
      bSeparateFillAlpha=false
      outlineMaterial=null
      outlineSize=0
      outlineColor=<rgba 0 0 0 1>
     }
    }
   }
   slot {class=ButtonSlot, properties {padding=<ltrb 2 2 2 2>}}
  }
  VerticalBox_92 {
   widgetClassPath=&VerticalBox
   parent=VerticalBox_89
   widget=VerticalBox_92
   properties {bOverride_Cursor=true}
   slot {class=VerticalBoxSlot, properties {padding=<ltrb 4 4 4 4>}}
  }
  HorizontalBox_269 {
   widgetClassPath=&HorizontalBox
   parent=VerticalBox_92
   widget=HorizontalBox_269
   slot {class=VerticalBoxSlot, properties {horizontalAlignment=HAlign_Left, verticalAlignment=VAlign_Top, padding=<ltrb 2 2 2 2>}}
  }
  Button_FilterSelection {
   widgetClassPath=&EditorUtilityButton
   parent=HorizontalBox_269
   widget=Button_FilterSelection
   slot {class=HorizontalBoxSlot, properties {padding=<ltrb 2 2 2 2>}}
  }
  TextBlock {
   widgetClassPath=&TextBlock
   parent=Button_FilterSelection
   widget=TextBlock
   properties {
    text=Filter Selected Actors
    font {
     bForceMonospaced=false
     bMaterialIsStencil=false
     fontMaterial=null
     fontObject=&Roboto
     letterSpacing=0
     monospacedWidth=1
     size=9
     skewAmount=0
     typefaceFontName=Regular
     outlineSettings {
      bApplyOutlineToDropShadows=false
      bMiteredCorners=false
      bSeparateFillAlpha=false
      outlineMaterial=null
      outlineSize=0
      outlineColor=<rgba 0 0 0 1>
     }
    }
   }
   slot {class=ButtonSlot, properties {padding=<ltrb 2 2 2 2>}}
  }
  Button_ReplaceMatsOnMesh {
   widgetClassPath=&EditorUtilityButton
   parent=HorizontalBox_269
   widget=Button_ReplaceMatsOnMesh
   slot {class=HorizontalBoxSlot, properties {horizontalAlignment=HAlign_Left, verticalAlignment=VAlign_Top, padding=<ltrb 2 2 2 2>}}
  }
  TextBlock_49 {
   widgetClassPath=&TextBlock
   parent=Button_ReplaceMatsOnMesh
   widget=TextBlock_49
   properties {
    text=Replace Materials on Mesh
    font {
     bForceMonospaced=false
     bMaterialIsStencil=false
     fontMaterial=null
     fontObject=&Roboto
     letterSpacing=0
     monospacedWidth=1
     size=9
     skewAmount=0
     typefaceFontName=Regular
     outlineSettings {
      bApplyOutlineToDropShadows=false
      bMiteredCorners=false
      bSeparateFillAlpha=false
      outlineMaterial=null
      outlineSize=0
      outlineColor=<rgba 0 0 0 1>
     }
    }
   }
   slot {class=ButtonSlot, properties {padding=<ltrb 2 2 2 2>}}
  }
  Button_AddOldPathFromMesh {
   widgetClassPath=&EditorUtilityButton
   parent=HorizontalBox_269
   widget=Button_AddOldPathFromMesh
   slot {class=HorizontalBoxSlot, properties {horizontalAlignment=HAlign_Left, verticalAlignment=VAlign_Top, padding=<ltrb 2 2 2 2>}}
  }
  TextBlock_7 {
   widgetClassPath=&TextBlock
   parent=Button_AddOldPathFromMesh
   widget=TextBlock_7
   properties {
    text=Get Path From Selected
    font {
     bForceMonospaced=false
     bMaterialIsStencil=false
     fontMaterial=null
     fontObject=&Roboto
     letterSpacing=0
     monospacedWidth=1
     size=9
     skewAmount=0
     typefaceFontName=Regular
     outlineSettings {
      bApplyOutlineToDropShadows=false
      bMiteredCorners=false
      bSeparateFillAlpha=false
      outlineMaterial=null
      outlineSize=0
      outlineColor=<rgba 0 0 0 1>
     }
    }
   }
   slot {class=ButtonSlot, properties {padding=<ltrb 2 2 2 2>}}
  }
  HorizontalBox_268 {
   widgetClassPath=&HorizontalBox
   parent=VerticalBox_92
   widget=HorizontalBox_268
   slot {class=VerticalBoxSlot, properties {padding=<ltrb 2 2 2 2>}}
  }
  EditorUtilityCheckBox_Save {
   widgetClassPath=&EditorUtilityCheckBox
   bIsVariable=true
   parent=HorizontalBox_268
   widget=EditorUtilityCheckBox_Save
   properties {checkedState=Checked}
   slot {class=HorizontalBoxSlot}
  }
  TextBlock_314 {
   widgetClassPath=&TextBlock
   parent=HorizontalBox_268
   widget=TextBlock_314
   properties {
    text=Save Asset if Modified
    font {
     bForceMonospaced=false
     bMaterialIsStencil=false
     fontMaterial=null
     fontObject=&Roboto
     letterSpacing=0
     monospacedWidth=1
     size=9
     skewAmount=0
     typefaceFontName=Regular
     outlineSettings {
      bApplyOutlineToDropShadows=false
      bMiteredCorners=false
      bSeparateFillAlpha=false
      outlineMaterial=null
      outlineSize=0
      outlineColor=<rgba 0 0 0 1>
     }
    }
   }
   slot {class=HorizontalBoxSlot, properties {verticalAlignment=VAlign_Center}}
  }
  HorizontalBox_1 {
   widgetClassPath=&HorizontalBox
   parent=VerticalBox_92
   widget=HorizontalBox_1
   slot {class=VerticalBoxSlot, properties {padding=<ltrb 2 2 2 2>}}
  }
  EditorUtilityCheckBox_Checkout {
   widgetClassPath=&EditorUtilityCheckBox
   bIsVariable=true
   parent=HorizontalBox_1
   widget=EditorUtilityCheckBox_Checkout
   properties {checkedState=Checked}
   slot {class=HorizontalBoxSlot}
  }
  TextBlock_2 {
   widgetClassPath=&TextBlock
   parent=HorizontalBox_1
   widget=TextBlock_2
   properties {
    text=Checkout Asset if Modified
    font {
     bForceMonospaced=false
     bMaterialIsStencil=false
     fontMaterial=null
     fontObject=&Roboto
     letterSpacing=0
     monospacedWidth=1
     size=9
     skewAmount=0
     typefaceFontName=Regular
     outlineSettings {
      bApplyOutlineToDropShadows=false
      bMiteredCorners=false
      bSeparateFillAlpha=false
      outlineMaterial=null
      outlineSize=0
      outlineColor=<rgba 0 0 0 1>
     }
    }
   }
   slot {class=HorizontalBoxSlot, properties {verticalAlignment=VAlign_Center}}
  }
  HorizontalBox_0 {
   widgetClassPath=&HorizontalBox
   parent=VerticalBox_89
   widget=HorizontalBox_0
   slot {class=VerticalBoxSlot, properties {padding=<ltrb 8 8 8 8>}}
  }
  VerticalBox_91 {
   widgetClassPath=&VerticalBox
   parent=HorizontalBox_0
   widget=VerticalBox_91
   slot {class=HorizontalBoxSlot, properties {size {sizeRule=Fill, value=1}}}
  }
  TextBlock_440 {
   widgetClassPath=&TextBlock
   parent=VerticalBox_91
   widget=TextBlock_440
   properties {
    text=Old Paths to Search
    font {
     bForceMonospaced=false
     bMaterialIsStencil=false
     fontMaterial=null
     fontObject=&Roboto
     letterSpacing=0
     monospacedWidth=1
     size=12
     skewAmount=0
     typefaceFontName=Regular
     outlineSettings {
      bApplyOutlineToDropShadows=false
      bMiteredCorners=false
      bSeparateFillAlpha=false
      outlineMaterial=null
      outlineSize=0
      outlineColor=<rgba 0 0 0 1>
     }
    }
   }
   slot {class=VerticalBoxSlot}
  }
  TextBlock_Paths {
   widgetClassPath=&TextBlock
   bIsVariable=true
   parent=VerticalBox_91
   widget=TextBlock_Paths
   properties {
    text=Text Block
    font {
     bForceMonospaced=false
     bMaterialIsStencil=false
     fontMaterial=null
     fontObject=&Roboto
     letterSpacing=0
     monospacedWidth=1
     size=12
     skewAmount=0
     typefaceFontName=Regular
     outlineSettings {
      bApplyOutlineToDropShadows=false
      bMiteredCorners=false
      bSeparateFillAlpha=false
      outlineMaterial=null
      outlineSize=0
      outlineColor=<rgba 0 0 0 1>
     }
    }
   }
   slot {class=VerticalBoxSlot}
  }
  VerticalBox {
   widgetClassPath=&VerticalBox
   parent=HorizontalBox_0
   widget=VerticalBox
   slot {class=HorizontalBoxSlot, properties {size {sizeRule=Fill, value=1}}}
  }
  TextBlock_4 {
   widgetClassPath=&TextBlock
   parent=VerticalBox
   widget=TextBlock_4
   properties {
    text=New Paths to Search
    font {
     bForceMonospaced=false
     bMaterialIsStencil=false
     fontMaterial=null
     fontObject=&Roboto
     letterSpacing=0
     monospacedWidth=1
     size=12
     skewAmount=0
     typefaceFontName=Regular
     outlineSettings {
      bApplyOutlineToDropShadows=false
      bMiteredCorners=false
      bSeparateFillAlpha=false
      outlineMaterial=null
      outlineSize=0
      outlineColor=<rgba 0 0 0 1>
     }
    }
   }
   slot {class=VerticalBoxSlot}
  }
  TextBlock_PathsNew {
   widgetClassPath=&TextBlock
   bIsVariable=true
   parent=VerticalBox
   widget=TextBlock_PathsNew
   properties {
    text=Text Block
    font {
     bForceMonospaced=false
     bMaterialIsStencil=false
     fontMaterial=null
     fontObject=&Roboto
     letterSpacing=0
     monospacedWidth=1
     size=12
     skewAmount=0
     typefaceFontName=Regular
     outlineSettings {
      bApplyOutlineToDropShadows=false
      bMiteredCorners=false
      bSeparateFillAlpha=false
      outlineMaterial=null
      outlineSize=0
      outlineColor=<rgba 0 0 0 1>
     }
    }
   }
   slot {class=VerticalBoxSlot}
  }
  TextBlock_Debug {
   widgetClassPath=&TextBlock
   bIsVariable=true
   parent=VerticalBox_89
   widget=TextBlock_Debug
   properties {
    text=Debug...
    font {
     bForceMonospaced=false
     bMaterialIsStencil=false
     fontMaterial=null
     fontObject=&Roboto
     letterSpacing=0
     monospacedWidth=1
     size=12
     skewAmount=0
     typefaceFontName=Regular
     outlineSettings {
      bApplyOutlineToDropShadows=false
      bMiteredCorners=false
      bSeparateFillAlpha=false
      outlineMaterial=null
      outlineSize=0
      outlineColor=<rgba 0 0 0 1>
     }
    }
   }
   slot {class=VerticalBoxSlot, properties {verticalAlignment=VAlign_Bottom, size {sizeRule=Fill, value=1}}}
  }
 ]
}
